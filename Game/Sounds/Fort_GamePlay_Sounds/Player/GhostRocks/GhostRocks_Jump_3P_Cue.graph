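# GhostRocks_Jump_3P_Cue is a player Sound Cue named for the third-person GhostRocks jump sound.
# It has no nodes and no root, so it plays no audio as saved.
# It still carries a halved volume, GhostRocks attenuation and the enemy third-person foley sound class.
# A concurrency override limits it to one instance per owner and blocks new plays while one is active.
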SoundCue /Game/Sounds/Fort_GamePlay_Sounds/Player/GhostRocks/GhostRocks_Jump_3P_Cue
root=null
nodes []
properties {
 attenuationSettings=/Game/Sounds/Attenuation/Fort_GhostRocks_Att
 bOverrideConcurrency=true
 maxDistance=4500
 soundClassObject=&Fort_Player_Foley_3P_Enemy_Athena
 volumeMultiplier=0.5
 concurrencyOverrides {bLimitToOwner=true, maxCount=1, resolutionRule=PreventNew}
}
