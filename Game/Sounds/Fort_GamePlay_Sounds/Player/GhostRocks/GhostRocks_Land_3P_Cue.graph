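# GhostRocks_Land_3P_Cue is a sound cue named for third-person GhostRocks landings, but it has no root or nodes, so it references no audio.
# It still carries playback settings: GhostRocks attenuation, the enemy third-person player foley sound class, reduced volume and a max distance.
# It references the shared footstep-landing concurrency set and also stores per-owner, stop-oldest concurrency override values.
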SoundCue /Game/Sounds/Fort_GamePlay_Sounds/Player/GhostRocks/GhostRocks_Land_3P_Cue
root=null
nodes []
properties {
 attenuationSettings=/Game/Sounds/Attenuation/Fort_GhostRocks_Att
 maxDistance=4500
 soundClassObject=&Fort_Player_Foley_3P_Enemy_Athena
 volumeMultiplier=0.7
 concurrencyOverrides {bLimitToOwner=true, maxCount=4, resolutionRule=StopOldest}
 concurrencySet [/Game/Sounds/Concurrency_Rules/Athena_Footstep_Land_Conc]
}
